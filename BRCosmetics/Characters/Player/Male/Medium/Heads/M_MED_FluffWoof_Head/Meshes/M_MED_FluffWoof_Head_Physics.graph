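# Physics asset for the FluffWoof medium male head cosmetic, with capsule bodies for the upper spine, clavicles, neck, head, face attach and hair bones.
# Locked joints chain the clavicles, neck and head to spine_05, while face attach and the central hair root have free angular joints to the head.
# The other base hair bodies lock to that root, and each carries a dynamic hair tip that swings within limits with zero twist.
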PhysicsAsset /BRCosmetics/Characters/Player/Male/Medium/Heads/M_MED_FluffWoof_Head/Meshes/M_MED_FluffWoof_Head_Physics
bodies {
 base_hair_FC_00 [{shapeType=Capsule, length=0.505, radius=0.505, shapeName=base_hair_FC_00_capsule, rotation=<rot 0 -90 0>}]
 base_hair_FC_01 [{shapeType=Capsule, length=0.505, radius=0.505, shapeName=base_hair_FC_01_capsule, rotation=<rot 0 -90 0>}]
 base_hair_FL_00 [{shapeType=Capsule, length=0.505, radius=0.505, shapeName=base_hair_FL_00_capsule, rotation=<rot 0 -90 0>}]
 base_hair_FR_00 [{shapeType=Capsule, length=0.505, radius=0.505, shapeName=base_hair_FR_00_capsule, rotation=<rot 0 -90 0>}]
 deform_clavicle_l [{shapeType=Capsule, length=7.5, radius=8, shapeName=deform_clavicle_l_capsule, center=<1.8 -0.6 -2.6>, rotation=<rot 0 -90 115>}]
 deform_clavicle_r [{shapeType=Capsule, length=7.5, radius=8.5, shapeName=deform_clavicle_r_capsule, center=<-1.8 0.6 2.6>, rotation=<rot 0 -90 115>}]
 dyn_hair_FC_01 [{shapeType=Capsule, length=3, radius=3, shapeName=dyn_hair_FC_01_capsule, center=<1 0 0>}]
 dyn_hair_FL_00 [{shapeType=Capsule, length=2, radius=1.6, shapeName=dyn_hair_FL_00_capsule, center=<1 0 0>, rotation=<rot 45 0 0>}]
 dyn_hair_FR_00 [{shapeType=Capsule, length=1, radius=2, shapeName=dyn_hair_FR_00_capsule, center=<1 0 0>, rotation=<rot -45 0 0>}]
 faceAttach [{shapeType=Capsule, length=7.9455, radius=5.2938, center=<0.4529 8.029 -0.3442>, rotation=<rot 0 89.5515 -89.7354>}]
 head [{shapeType=Capsule, length=13.6632, radius=12.5972, shapeName=head_capsule, center=<8.3941 4.8519 -1.5151>, rotation=<rot 7.2042 -90 7.9778>}]
 neck_01 [{
   shapeType=Capsule
   length=5.2162
   radius=2.7552
   shapeName=neck_01_capsule
   center=<5.1507 -2.1681 -0.6205>
   rotation=<rot -3.6914 -90 4.8202>
  }]
 neck_02 [{
   shapeType=Capsule
   length=0.505
   radius=0.505
   shapeName=neck_02_capsule
   center=<2.4493 0.8735 -1.8753>
   rotation=<rot 0 -156.0934 -78.9877>
  }]
 spine_05 [{shapeType=Capsule, length=5.5, radius=10.5, shapeName=spine_05_capsule, center=<3.8344 -0.8897 0>, rotation=<rot 0 -90 0>}]
}
constraints [
 {
  bone1Name=base_hair_FC_01
  bone2Name=base_hair_FC_00
  swing1LimitDegrees=45
  swing1Motion=Locked
  swing2LimitDegrees=45
  swing2Motion=Locked
  twistLimitDegrees=45
  twistMotion=Locked
 }
 {bone1Name=dyn_hair_FC_01, bone2Name=base_hair_FC_01, swing1LimitDegrees=45, swing2LimitDegrees=45, twistLimitDegrees=0}
 {
  bone1Name=base_hair_FL_00
  bone2Name=base_hair_FC_00
  swing1LimitDegrees=45
  swing1Motion=Locked
  swing2LimitDegrees=45
  swing2Motion=Locked
  twistLimitDegrees=45
  twistMotion=Locked
 }
 {bone1Name=dyn_hair_FL_00, bone2Name=base_hair_FL_00, swing1LimitDegrees=45, swing2LimitDegrees=45, twistLimitDegrees=0}
 {
  bone1Name=base_hair_FR_00
  bone2Name=base_hair_FC_00
  swing1LimitDegrees=45
  swing1Motion=Locked
  swing2LimitDegrees=45
  swing2Motion=Locked
  twistLimitDegrees=45
  twistMotion=Locked
 }
 {bone1Name=dyn_hair_FR_00, bone2Name=base_hair_FR_00, swing1LimitDegrees=45, swing2LimitDegrees=45, twistLimitDegrees=0}
 {
  bone1Name=deform_clavicle_l
  bone2Name=spine_05
  swing1LimitDegrees=45
  swing1Motion=Locked
  swing2LimitDegrees=45
  swing2Motion=Locked
  twistLimitDegrees=45
  twistMotion=Locked
 }
 {
  bone1Name=deform_clavicle_r
  bone2Name=spine_05
  swing1LimitDegrees=45
  swing1Motion=Locked
  swing2LimitDegrees=45
  swing2Motion=Locked
  twistLimitDegrees=45
  twistMotion=Locked
 }
 {
  bone1Name=faceAttach
  bone2Name=head
  swing1LimitDegrees=45
  swing1Motion=Free
  swing2LimitDegrees=45
  swing2Motion=Free
  twistLimitDegrees=45
  twistMotion=Free
 }
 {
  bone1Name=base_hair_FC_00
  bone2Name=head
  swing1LimitDegrees=45
  swing1Motion=Free
  swing2LimitDegrees=45
  swing2Motion=Free
  twistLimitDegrees=45
  twistMotion=Free
 }
 {
  bone1Name=neck_01
  bone2Name=spine_05
  swing1LimitDegrees=45
  swing1Motion=Locked
  swing2LimitDegrees=45
  swing2Motion=Locked
  twistLimitDegrees=45
  twistMotion=Locked
 }
 {
  bone1Name=neck_02
  bone2Name=neck_01
  swing1LimitDegrees=45
  swing1Motion=Locked
  swing2LimitDegrees=45
  swing2Motion=Locked
  twistLimitDegrees=45
  twistMotion=Locked
 }
 {
  bone1Name=head
  bone2Name=neck_02
  swing1LimitDegrees=45
  swing1Motion=Locked
  swing2LimitDegrees=45
  swing2Motion=Locked
  twistLimitDegrees=45
  twistMotion=Locked
 }
]
